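; Blueprint for a laser grid actor in heist bases that can switch off its laser during a lockdown.
; OnLockdownInitiated reads the GetHotfix_DisableOnLockdown scalable float as a bool; if true, it disables the laser and returns true, otherwise it returns false.
; The construction script only calls its parent version.
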(fn OnLockdownInitiated
  ;[0.1349 0.05632 0.1667] OnLockdownInitiated: Disable the laser actor.
  (if (ScalableFloat|GetValueasBool (Variables|Default|GetHotfix_DisableOnLockdown))
    (Laser|DisableLaser)
    (return true)
    (else
      (return false))))

(fn ConstructionScript
  (|CallParentFunction))
