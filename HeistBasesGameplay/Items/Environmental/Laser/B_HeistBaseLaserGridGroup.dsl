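; Laser grid group for heist bases that ties laser grids to a set of security gates.
; On the server, once playlist data is ready, it marks each valid gate to ignore lockdown.
; When a laser grid in the group is triggered, the server waits a brief delay, then requests that every valid gate close.
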(event EventBeginPlay
  ;[0.03793 0.1823 0.05083] BeginPlay: The security gates need to be marked to ignore lockdown.
  (|CallParentFunction)
  (Utilities|SwitchNetMode
    (:Server
      (Class|FortAsyncActionCurrentPlaylistReady|PlaylistDataReadyAsync
        (:OnReady
          (for _array_element (Variables|Default|GetSecurityGates)
            (Utilities|IsValid _array_element
              (:Is%20Valid
                (Class|BHeistBasesSecurityGate|MarkIgnoreLockdown _array_element)))))))))

(event EventHandleLaserGridTriggered
  (params (LaserGrid :type Building%20Gameplay%20Actor%20Laser%20Grid%20Object%20Reference) (LaserConnectionData :type Laser%20Grid%20Connection%20Entry%20Structure%20%28by%20ref%29) (TriggeringActor :type Actor%20Object%20Reference))
  ;[0.07853 0.02167 0.1302] HandleLaserGridTriggered: Laser grid is triggered, close the gates after a small delay.
  (|CallParentFunction LaserGrid LaserConnectionData TriggeringActor)
  (Utilities|SwitchNetMode
    (:Server
      (Utilities|FlowControl|Delay 0.1) ; very small delay b/w triggering laser and gates closing.
      (for _array_element (Variables|Default|GetSecurityGates)
        (Utilities|IsValid _array_element
          (:Is%20Valid
            (Class|BHeistBasesSecurityGate|RequestCloseGate _array_element)))))))

(fn ConstructionScript)
